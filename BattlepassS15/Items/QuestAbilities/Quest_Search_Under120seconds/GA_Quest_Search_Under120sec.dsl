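; Transient quest gameplay ability for a Season 15 timed search quest (under 120 seconds, per its name), reporting success or expiry.
; On activation, the server binds quest manager completed/failed events and the pawn's transient quest time-expired event.
; Matching completion shows a HUD success message and removes the quest; expiry shows a failure message.
; All outcomes call Cleanup, which unbinds events without ending the ability, fetching the manager as "Invalid" rather than "Athena".
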(event Ability|EventFailedToActivatePassiveAbility
  (params (ActorInfo :type Gameplay%20Ability%20Actor%20Info%20Structure))
  (|CallParentFunction))

(event Ability|EventActivateAbility
  (bind _owning_player (Variables|Default|GetOwning%20Player))
  (bind _returnvalue (Fortnite|GetFortPlayerControllerfromActor _owning_player))
  (bind _returnvalue_2 (Quests|GetQuestManager _returnvalue "Athena"))
  (bind _returnvalue_3 (Class|FortQuestManager|GetQuestManagerComponent _returnvalue_2 "/Script/FortniteGame.FortQuestManagerComponent_Athena"))
  ;[0 0 0] Activate
  (bind _asfort_player_pawn_athena (Utilities|Casting|CastToFortPlayerPawnAthena (Ability|GetAvatarActorfromActorInfo))
    (:then
      (Variables|Default|SetOwning%20Player _asfort_player_pawn_athena)
      (if (Networking|IsServer)
        (Quests|BindEventtoOnUrgentTransientQuestCompleted _returnvalue_3 (EventDispatchers|CreateEvent :!event Custom|Quest%20Completed))
        (Quests|BindEventto%28DEPRECATED%29OnUrgentTransientQuestFailed _returnvalue_3 (EventDispatchers|CreateEvent :!event Custom|Cleanup%20On%20Quest%20Ended))
        (EventDispatchers|AssignOnUrgentQuestTimeExpired (Class|FortPlayerPawn|GetTransientQuestsComponent (Variables|Default|GetOwning%20Player)) (EventDispatchers|CreateEvent :!event Custom|Quest%20Failed))))
    (:CastFailed
      (Ability|EndAbility))))

(event Custom|Quest%20Completed
  (params (TransientQuestItem :type Fort%20Quest%20Item%20Object%20Reference))
  (bind _questitemdefinition (Variables|FortTransientQuestGameplayAbility|GetQuestItemDefinition))
  ;[0 0 0] Quest Completed
  (if (AssetManager|Equal(PrimaryAssetId) (AssetManager|GetPrimaryAssetIdfromObject (Quest|GetQuestDefinition TransientQuestItem)) (AssetManager|GetPrimaryAssetIdfromObject _questitemdefinition))
    (bind _asfort_urgent_quest_item_definition (Utilities|Casting|CastToFortUrgentQuestItemDefinition _questitemdefinition))
    (CallFunction|DisplaySuccessMessage (Class|FortUrgentQuestItemDefinition|GetUrgentQuestData _asfort_urgent_quest_item_definition))
    (Class|FortControllerComponentTransientQuests|CleanupCompletedQuest (Class|FortPlayerPawn|GetTransientQuestsComponent (Variables|Default|GetOwning%20Player)) _questitemdefinition)
    (CallFunction|Cleanup)))

(event Custom|Quest%20Failed
  (params (UrgentQuestData :type Urgent%20Quest%20Data%20Structure%20%28by%20ref%29))
  ;[0 0 0] Quest Failed
  (CallFunction|DisplayFailureMessageTimeExpired UrgentQuestData)
  (CallFunction|Cleanup))

(event Custom|Cleanup%20On%20Quest%20Ended
  (params (TransientQuestItem :type Fort%20Quest%20Item%20Object%20Reference))
  ;[0 0 0] Cleanup if the quest ended for a reason other than the time expired
  (CallFunction|Cleanup))

(event Custom|DisplaySuccessMessage
  (params (Quest :type Urgent%20Quest%20Data%20Structure))
  (bind (_eventtag _visualstyletag _totaleventtime) (Utilities|Struct|BreakUrgentQuestData Quest))
  ;[0 0 0] DisplaySuccessMessage
  (Class|AthenaHUDContext|ShowUrgentQuestMessage (BlueprintContext|GetAthenaHUDContext) (Utilities|Struct|MakeUrgentQuestData _eventtag "(GameplayTags=)" _totaleventtime true false false false false true "" (Variables|Default|GetText_QuestSuccess) (Variables|Default|GetSuccessReason)))
  (Development|PrintString "Objective Complete"))

(event Custom|DisplayFailureMessage_TimeExpired
  (params (Quest :type Urgent%20Quest%20Data%20Structure))
  (bind (_eventtag _visualstyletag _totaleventtime) (Utilities|Struct|BreakUrgentQuestData Quest))
  ;[0 0 0] DisplayFailureMessage_TimeExpired
  (Class|AthenaHUDContext|ShowUrgentQuestMessage (BlueprintContext|GetAthenaHUDContext) (Utilities|Struct|MakeUrgentQuestData _eventtag "(GameplayTags=)" _totaleventtime true false false false false true "" (Variables|Default|GetText_QuestFailed) (Variables|Default|GetText_TimeExpired))))

(event Custom|Cleanup
  (bind _owning_player (Variables|Default|GetOwning%20Player))
  (bind _returnvalue (Fortnite|GetFortPlayerControllerfromActor _owning_player))
  (bind _returnvalue_2 (Quests|GetQuestManager _returnvalue "Invalid"))
  (bind _returnvalue_3 (Class|FortQuestManager|GetQuestManagerComponent _returnvalue_2 "/Script/FortniteGame.FortQuestManagerComponent_Athena"))
  ;[0 0 0] Cleanup Binds
  (Quests|UnbindEventfromOnUrgentTransientQuestCompleted _returnvalue_3 (EventDispatchers|CreateEvent :!event Custom|Quest%20Completed))
  (Quests|UnbindEventfrom%28DEPRECATED%29OnUrgentTransientQuestFailed _returnvalue_3 (EventDispatchers|CreateEvent :!event Custom|Cleanup%20On%20Quest%20Ended))
  (EventDispatchers|UnbindEventfromOnUrgentQuestTimeExpired (Class|FortPlayerPawn|GetTransientQuestsComponent (Variables|Default|GetOwning%20Player)) (EventDispatchers|CreateEvent :!event Custom|Quest%20Failed))
  (Development|PrintString "Cleanup Ability"))
